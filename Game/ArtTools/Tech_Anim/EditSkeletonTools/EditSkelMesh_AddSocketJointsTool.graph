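# Editor utility widget that adds attachment joints to a batch of skeletal meshes and moves sockets onto them.
# For each target path, SkeletonModifier copies missing listed bones from a source mesh, commits, and prints a warning on unexpected bone counts.
# It then reparents spine_05 sockets whose names contain listed entries to the bone at the same list index.
# Socket transforms are recomputed relative to each bone's local transform to offset the parent change.
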EditorUtilityWidgetBlueprint /Game/ArtTools/Tech_Anim/EditSkeletonTools/EditSkelMesh_AddSocketJointsTool
graphs [
 EventGraph {nodes [
   "On Clicked (TransferJointsButton)"() {is=title K2Node_ComponentBoundEvent}
   > n1 = Sequence!(execute) -> (then_0, then_1, then_2:Exec, then_3:Exec, then_4:Exec, then_5:Exec)
   n2 = Get Text(self:"Text Box (Multi- Line) Object Reference"=Get BoneList_TextBox() -> (BoneList_TextBox:Editor Utility Multi Line Editable Text Box Object Reference))
   > Set BonesInputStrings(BonesInputStrings:Array of Strings=ParseIntoArray(SourceString="To String (Text)"(InText=@n2), Delimiter="\r\n")) -> (Output_Get:Array of Strings)
   n3 = Get Text(self=Get SocketsList_TextBox() -> (SocketsList_TextBox:Editor Utility Multi Line Editable Text Box Object Reference))
   Set SocketsInputStrings(execute=@n1.then_1, SocketsInputStrings:Array of Strings=ParseIntoArray(SourceString="To String (Text)"(InText=@n3), Delimiter="\r\n")) -> (Output_Get:Array of Strings)
   n4 = Get Text(self=Get SkelMesh_TextBox() -> (SkelMesh_TextBox:Editor Utility Multi Line Editable Text Box Object Reference))
   Set SkelMeshesToOperateOn(execute=@n1.then_2, SkelMeshesToOperateOn:Array of Strings=ParseIntoArray(SourceString="To String (Text)"(InText=@n4), Delimiter="\r\n")) -> (Output_Get:Array of Strings)
   n5 = Construct Object from Class(execute=@n1.then_3, Class=/Script/SkeletalMeshModifiers.SkeletonModifier) -> (ReturnValue:Skeleton Modifier Object Reference)
   > Is Valid(InputObject=@n5)
   > n6 = Set SourceSkeletonModifier(SourceSkeletonModifier:Skeleton Modifier Object Reference=@n5) -> (Output_Get:Skeleton Modifier Object Reference)
   > SetSkeletalMesh!(
    execute
    self:Skeleton Modifier Object Reference=@n6
    InSkeletalMesh:Skeletal Mesh Object Reference=Get SourceSkelMesh() -> (SourceSkelMesh:Skeletal Mesh Object Reference)
   ) -> (then, ReturnValue:bool)
   n7 = For Each Loop(Exec=@n1.then_4, Array=Get SkelMeshesToOperateOn() -> (SkelMeshesToOperateOn:Array of Strings))
   > n8 = Set SkelMeshPath(SkelMeshPath:String=@n7) -> (Output_Get:String)
   > n9 = LoadAsset!(
    execute
    self:Editor Asset Subsystem Object Reference=Get EditorAssetSubsystem() -> (ReturnValue:Editor Asset Subsystem Object Reference) {is=title K2Node_GetEditorSubsystem}
    AssetPath=@n8
   ) -> (then, ReturnValue)
   > n10 = Cast To SkeletalMesh!(execute, Object=@n9) -> (then, CastFailed, AsSkeletal Mesh:Skeletal Mesh Object Reference) {is=title K2Node_DynamicCast}
   > Set TargetSkelMesh(TargetSkelMesh:Skeletal Mesh Object Reference=@n10) -> (Output_Get:Skeletal Mesh Object Reference)
   > n11 = Sequence!(execute) -> (then_0, then_1, then_2, then_3, then_4)
   > PrintString(InString=Checking for missing bones and adding them:::)
   > n12 = Construct Object from Class(Class=&SkeletonModifier)
   > Is Valid(InputObject=@n12)
   > n13 = Set TargetSkeletonModifier(TargetSkeletonModifier:Skeleton Modifier Object Reference=@n12) -> (Output_Get:Skeleton Modifier Object Reference)
   > SetSkeletalMesh(self=@n13, InSkeletalMesh=Get TargetSkelMesh() -> (TargetSkelMesh:Skeletal Mesh Object Reference))
   n14 = Length(TargetArray:Array of Names=GetAllBoneNames!(self:Skeleton Modifier Object Reference=@n13) -> (ReturnValue:Array of Names))
   > Set TargetSkelMesh_OrigBoneCount(TargetSkelMesh_OrigBoneCount:int=@n14) -> (Output_Get:int=0)
   > PrintString(InString=Append(A="To String (Integer)"(InInt=@n14), B=" bones in the Target SkelMesh. "))
   > n15 = For Each Loop(Array=Get BonesInputStrings() -> (BonesInputStrings:Array of Strings))
   n16 = Get TargetSkeletonModifier() -> (TargetSkeletonModifier:Skeleton Modifier Object Reference)
   CommitSkeletonToSkeletalMesh!(execute=@n15.Completed, self:Skeleton Modifier Object Reference=@n16) -> (then, ReturnValue:bool)
   n17 = Length(TargetArray=GetAllBoneNames(self=@n16))
   > Set TargetSkelMesh_NewBoneCount(TargetSkelMesh_NewBoneCount:int=@n17) -> (Output_Get:int=0)
   > PrintString(InString=Append(A="To String (Integer)"(InInt=@n17), B=" bones in the Target SkelMesh after adding bones. "))
   n18 = Get TargetSkelMesh_OrigBoneCount() -> (TargetSkelMesh_OrigBoneCount:int=0)
   n19 = int + int(A=@n18, B=2)
   n20 = Get TargetSkelMesh_NewBoneCount() -> (TargetSkelMesh_NewBoneCount:int=0)
   > Branch(Condition=OR Boolean(A="integer > integer"(A=@n20, B=@n19), B="integer < integer"(A=@n20, B=@n18)))
   > PrintString(InString="BONE COUNT CHANGED TOO MUCH!!!!!!!!", bPrintToScreen=false, TextColor=(R=1,G=0,B=0,A=1))
   Set BoneToTransfer(execute=@n15.LoopBody, BoneToTransfer:String=@n15) -> (Output_Get:String)
   n21 = GetAllBoneNames(self=Get SourceSkeletonModifier() -> (SourceSkeletonModifier:Skeleton Modifier Object Reference))
   n22 = String To Name(InString=Get BoneToTransfer() -> (BoneToTransfer:String))
   > n23 = Branch(Condition="integer >= integer"(A=Find Item!(TargetArray:Array of Names=@n21, ItemToFind:"Name (by ref)"=@n22) -> (ReturnValue)))
   > PrintString(InString=Append!(A="\"", B=Get BoneToTransfer(), C:String="\" FOUND in Source SkelMesh") -> (ReturnValue))
   n24 = GetAllBoneNames(self=Get TargetSkeletonModifier())
   n25 = String To Name(InString=Get BoneToTransfer())
   > n26 = Branch(Condition="integer >= integer"(A=Find Item(TargetArray=@n24, ItemToFind=@n25)))
   > PrintString(InString=Append(A="\"", B=Get BoneToTransfer(), C="\" ALREADY EXISTS in the Target SkelMesh.  Skipped"))
   n27 = String To Name(InString=Get BoneToTransfer())
   n28 = Get SourceSkeletonModifier()
   AddBone!(
    execute=@n26.else
    self:Skeleton Modifier Object Reference=Get TargetSkeletonModifier()
    InBoneName:Name=@n27
    InParentName:Name=GetParentName!(self:Skeleton Modifier Object Reference=@n28, InBoneName:Name=@n27) -> (ReturnValue:Name=None)
    InTransform:"Transform (by ref)"=GetBoneTransform!(self:Skeleton Modifier Object Reference=@n28, InBoneName:Name=@n27, bGlobal:bool=false) -> (ReturnValue:Transform)
   ) -> (then, ReturnValue:bool)
   > PrintString(InString=Append(A="\"", B=Get BoneToTransfer(), C="\" ADDED to the Target SkelMesh"))
   PrintString(execute=@n23.else, InString=Append(A="\"", B=Get BoneToTransfer(), C="\" NOT FOUND in the Source SkelMesh"))
   PrintString(execute=@n11.then_1, InString=Finding all sockets on spine_05:::)
   > Clear(TargetArray:Array of Names=Get AllSocketNames() -> (AllSocketNames:Array of Names))
   n29 = Get TargetSkelMesh()
   > n30 = For Loop(LastIndex=int - int(A=NumSockets!(self:Skeletal Mesh Object Reference=@n29) -> (ReturnValue:int=0)))
   n31 = GetSocketByIndex!(self:Skeletal Mesh Object Reference=Get TargetSkelMesh(), Index:int=@n30) -> (ReturnValue:Skeletal Mesh Socket Object Reference)
   n32 = Get BoneName!(self:Skeletal Mesh Socket Object Reference=@n31) -> (BoneName:Name=None)
   n33 = Get SocketName!(self:Skeletal Mesh Socket Object Reference=@n31) -> (SocketName:Name=None)
   > PrintString(InString=Append(A="To String (Name)"(InName=@n33), B="--  PARENT JOINT:  ", C="To String (Name)"(InName=@n32)))
   > Branch(Condition="Equal (Name)"(A=@n32, B=spine_05))
   > Add(TargetArray:Array of Names=Get AllSocketNames(), NewItem:"Name (by ref)"=@n33) {is=title K2Node_CallArrayFunction}
   PrintString(execute=@n11.then_3, InString=Moving sockets to their appropriate locations:::)
   > n34 = For Each Loop(Array:Array of Names=Get AllSocketNames()) -> (Array Element:Name)
   > n35 = For Each Loop(Array:Array of Strings=Get SocketsInputStrings() -> (SocketsInputStrings:Array of Strings)) -> (Array Element:String)
   n36 = "To String (Name)"(InName=@n34)
   > Branch(Condition=Contains(SearchIn=@n36, Substring=@n35))
   > Set SocketToMove(SocketToMove:String=@n36) -> (Output_Get:String)
   > Set BoneToTransfer(BoneToTransfer="Get (a ref)"(Array:Array of Strings=Get BonesInputStrings(), Dimension 1=@n35.Array Index) -> (Output:"String (by ref)"))
   n37 = FindSocket!(
    self:Skinned Asset Object Reference=Get TargetSkelMesh()
    InSocketName:Name=String To Name(InString=Get SocketToMove() -> (SocketToMove:String))
   ) -> (ReturnValue:Skeletal Mesh Socket Object Reference)
   > Set SocketToMove_PreTransform(SocketToMove_PreTransform:Transform=GetSocketLocalTransform!(self:Skeletal Mesh Socket Object Reference=@n37) -> (ReturnValue:Transform)) -> (Output_Get:Transform)
   n38 = String To Name(InString=Get BoneToTransfer())
   > SetSocketParent!(
    execute
    self:Skeletal Mesh Socket Object Reference=@n37
    InSkeletalMesh:Skeletal Mesh Object Reference=Get TargetSkelMesh()
    InBoneName:Name=@n38
   ) -> (then)
   > n39 = GetLocalSpaceBoneTransform!(execute, SkeletalMesh:Skeletal Mesh Object Reference=Get TargetSkelMesh(), BoneName:Name=@n38) -> (then, Valid:bool, ReturnValue:Transform)
   > n40 = Set BoneToTransferTransform(BoneToTransferTransform:Transform=@n39.ReturnValue) -> (Output_Get:Transform)
   > SetSocketLocalTransform!(
    execute
    self:Skeletal Mesh Socket Object Reference=@n37
    InTransform:Transform=MakeRelativeTransform!(A:"Transform (by ref)"=Get SocketToMove_PreTransform() -> (SocketToMove_PreTransform:Transform), RelativeTo:"Transform (by ref)"=@n40) -> (ReturnValue:Transform)
   ) -> (then)
   PrintString(execute=@n11.then_4, InString="-----------------------------------------\r\n\r\n")
   PrintString(execute=@n11.then_2, InString=ALL spine_05 SOCKETS)
   > n41 = For Each Loop(Array:Array of Names=Get AllSocketNames()) -> (Array Element:Name)
   > PrintString(InString="To String (Name)"(InName=@n41))
   Event PreConstruct()
   > SetObject(self=Get DetailsView() -> (DetailsView:Details View Object Reference), NewObject=Self-Reference())
  ]}
 GetAllSockets {nodes [
   n1 = GetAllSockets!() -> (then, Skeletal Mesh:Skeletal Mesh Object Reference) {is=title K2Node_FunctionEntry}
   n2 = Make Array("[0]":Skeletal Mesh Socket Object Reference) -> (Array:Array of Skeletal Mesh Socket Object References)
   > Remove Index!(execute, TargetArray:Array of Skeletal Mesh Socket Object References=@n2, IndexToRemove:int=0) -> (then)
   n3 = NumSockets(self=@n1) -> (ReturnValue=0)
   > n4 = For Loop(LastIndex=int - int(A=@n3))
   Return Node!(execute=@n4.Completed, ReturnValue:Array of Skeletal Mesh Socket Object References=@n2, NumSockets:int=@n3)
   > Add(
    TargetArray:Array of Skeletal Mesh Socket Object References=@n2
    NewItem:"Skeletal Mesh Socket Object Reference (by ref)"=GetSocketByIndex(self=@n1, Index=@n4)
   )
  ]}
]
variables [
 AllSocketNames {type="array<string>", default []}
 BonesInputStrings {type="array<string>", default []}
 BoneToTransfer {type=string, default=""}
 BoneToTransferTransform {type=Transform, default {location=<0 0 0>, rotation=<rot 0 0 0>, scale=<1 1 1>}}
 SkelMeshesToOperateOn {type="array<string>", default []}
 SkelMeshPath {type=string, default=""}
 SocketsInputStrings {type="array<string>", default []}
 SocketsToMoveALL {type="array<string>", default []}
 SocketToMove {type=string, default=""}
 SocketToMove_PreTransform {type=Transform, default {location=<0 0 0>, rotation=<rot 0 0 0>, scale=<1 1 1>}}
 SourceSkeletonModifier {type=SkeletonModifier, default=null}
 SourceSkelMesh {type=SkeletalMesh, default=null}
 TargetSkeletonModifier {type=SkeletonModifier, default=null}
 TargetSkelMesh {type=SkeletalMesh, default=null}
 TargetSkelMesh_NewBoneCount {type=integer}
 TargetSkelMesh_OrigBoneCount {type=integer}
 TargetSocketLocation {type=Vector}
]
widgets {
 info {rootWidgetClass=&StackBox, widgetCount=11}
 widgets [
  StackBox_0 {widgetClassPath=&StackBox, widget=StackBox_0}
  VerticalBox_47 {
   widgetClassPath=&VerticalBox
   parent=StackBox_0
   widget=VerticalBox_47
   slot {class=StackBoxSlot, properties {size {sizeRule=Fill, value=1}}}
  }
  DetailsView {
   widgetClassPath=&DetailsView
   parent=VerticalBox_47
   widget=DetailsView
   properties {
    toolTipText="Enter the Source Skeletal Mesh here.\r\nThis is the mesh that it will pull the bones from your \"Bones List\" below to use.  \r\n\r\nexamples: \r\n/Script/Engine.SkeletalMesh'/Game/Characters/Player/Male/Medium/Base/SK_M_MALE_Base_Skeleton'\r\n\r\n/Script/Engine.SkeletalMesh'/Game/Characters/Player/Female/Medium/Base/SK_M_Female_Base_Skeleton'\r\n"
    categoriesToShow [SkelMeshes]
   }
   slot {class=VerticalBoxSlot}
  }
  VerticalBox_83 {widgetClassPath=&VerticalBox, parent=VerticalBox_47, widget=VerticalBox_83, slot {class=VerticalBoxSlot}}
  BoneList_TextBox {
   widgetClassPath=/Script/Blutility.EditorUtilityMultiLineEditableTextBox
   bIsVariable=true
   parent=VerticalBox_83
   widget=BoneList_TextBox
   properties {
    hintText="BONES LIST\r\nexamples:\r\nattach_backpack\r\nattach_cape"
    toolTipText="Copy and Paste your bone list from a selection of bones in your Source SkelMesh. \r\n\r\nThese are the bones that will be transfered onto each Target SkelMesl in the Target SkelMeshes list below.\r\n\r\nexamples:\r\nattach_backpack\r\nattach_cape\r\n"
   }
   slot {class=VerticalBoxSlot, properties {size {sizeRule=Fill, value=1}}}
  }
  SocketsList_TextBox {
   widgetClassPath=&EditorUtilityMultiLineEditableTextBox
   bIsVariable=true
   parent=VerticalBox_83
   widget=SocketsList_TextBox
   properties {
    hintText="SOCKETS LIST\r\nexamples:\r\nBackpack_BR\r\ncape"
    toolTipText="Copy and Paste your socket list from a selection of sockets in your SkelMesh. \r\n\r\nThese are the sockets that will be moved onto each bone in the Bones List on the Target SkelMesl in the Target SkelMeshes list below.\r\nThe order you place these sockets matters relative to the Bones List.\r\n\r\nexamples:\r\nBackpack_BR\r\ncape"
   }
   slot {class=VerticalBoxSlot, properties {size {sizeRule=Fill, value=1}}}
  }
  Spacer_83 {widgetClassPath=&Spacer, parent=VerticalBox_47, widget=Spacer_83, properties {size=<1 5>}, slot {class=VerticalBoxSlot}}
  SkelMesh_TextBox {
   widgetClassPath=&EditorUtilityMultiLineEditableTextBox
   bIsVariable=true
   parent=VerticalBox_47
   widget=SkelMesh_TextBox
   properties {
    hintText="TARGET SKEL MESHES LIST\r\nexamples:\r\n/BRCosmetics/Characters/Player/Male/Medium/Bodies/M_MED_Handyman/Meshes/M_MED_Handyman\r\n/BRCosmetics/Characters/Player/Male/Medium/Bodies/M_MED_Solstice/Meshes/M_MED_Solstice\r\n"
    toolTipText="The Target Meshes list uses the Object Path for the meshes you want to operate on.\r\nYou can find the Object Path by right clicking on the mesh in the Content Browser"
   }
   slot {class=VerticalBoxSlot, properties {size {sizeRule=Fill, value=1}}}
  }
  HorizontalBox_76 {widgetClassPath=&HorizontalBox, parent=VerticalBox_47, widget=HorizontalBox_76, slot {class=VerticalBoxSlot}}
  TransferJointsButton {
   widgetClassPath=&EditorUtilityButton
   parent=HorizontalBox_76
   widget=TransferJointsButton
   slot {class=HorizontalBoxSlot, properties {padding=<ltrb 10 10 10 10>, size {sizeRule=Fill, value=1}}}
  }
  CommonTextBlock_53 {
   widgetClassPath=&CommonTextBlock
   parent=TransferJointsButton
   widget=CommonTextBlock_53
   properties {
    style=&TextStyle-Base_C
    text=Transfer Joints
    font {
     bForceMonospaced=false
     bMaterialIsStencil=false
     fontMaterial=null
     fontObject=&NotoSans
     letterSpacing=0
     monospacedWidth=1
     size=24
     skewAmount=0
     typefaceFontName=NotoSans-Regular
     outlineSettings {
      bApplyOutlineToDropShadows=false
      bMiteredCorners=false
      bSeparateFillAlpha=false
      outlineMaterial=null
      outlineSize=0
      outlineColor=<rgba 0 0 0 1>
     }
    }
    shadowOffset=<0 0>
   }
   slot {class=ButtonSlot}
  }
 ]
}
