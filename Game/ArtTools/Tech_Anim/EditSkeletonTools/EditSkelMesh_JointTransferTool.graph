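# Editor Utility Widget that transfers a pasted list of joints from a source skeletal mesh to a target one.
# A Details View exposes both meshes; the button splits the text by line and creates a SkeletonModifier for each mesh.
# Listed bones found in the source are added to the target, or given the source local transform if already present under the same parent.
# The target skeleton is committed after the loop, with results printed per bone.
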EditorUtilityWidgetBlueprint /Game/ArtTools/Tech_Anim/EditSkeletonTools/EditSkelMesh_JointTransferTool
graphs [EventGraph {nodes [
   "On Clicked (TransferJointsButton)"() {is=title K2Node_ComponentBoundEvent}
   > n1 = Sequence()
   n2 = Get Text(self:"Text Box (Multi- Line) Object Reference"=Get BoneList_TextBox() -> (BoneList_TextBox:Editor Utility Multi Line Editable Text Box Object Reference))
   > Set BonesToTransfer(BonesToTransfer:Array of Strings=ParseIntoArray(SourceString="To String (Text)"(InText=@n2), Delimiter="\r\n")) -> (Output_Get:Array of Strings)
   n3 = Construct Object from Class(execute=@n1.then_1, Class=/Script/SkeletalMeshModifiers.SkeletonModifier) -> (ReturnValue:Skeleton Modifier Object Reference)
   > Is Valid(InputObject=@n3)
   > n4 = Set SourceSkeletonModifier(SourceSkeletonModifier:Skeleton Modifier Object Reference=@n3) -> (Output_Get:Skeleton Modifier Object Reference)
   > SetSkeletalMesh!(
    execute
    self:Skeleton Modifier Object Reference=@n4
    InSkeletalMesh:Skeletal Mesh Object Reference=Get SourceSkelMesh() -> (SourceSkelMesh:Skeletal Mesh Object Reference)
   ) -> (then, ReturnValue:bool)
   > n5 = Construct Object from Class(Class=&SkeletonModifier)
   > Is Valid(InputObject=@n5)
   > n6 = Set TargetSkeletonModifier(TargetSkeletonModifier:Skeleton Modifier Object Reference=@n5) -> (Output_Get:Skeleton Modifier Object Reference)
   > SetSkeletalMesh(self=@n6, InSkeletalMesh=Get TargetSkelMesh() -> (TargetSkelMesh:Skeletal Mesh Object Reference))
   > n7 = For Each Loop(Array=Get BonesToTransfer() -> (BonesToTransfer:Array of Strings))
   CommitSkeletonToSkeletalMesh!(
    execute=@n7.Completed
    self:Skeleton Modifier Object Reference=Get TargetSkeletonModifier() -> (TargetSkeletonModifier:Skeleton Modifier Object Reference)
   ) -> (then, ReturnValue:bool)
   Set BoneToTransfer(execute=@n7.LoopBody, BoneToTransfer:String=@n7) -> (Output_Get:String)
   n8 = GetAllBoneNames!(self:Skeleton Modifier Object Reference=Get SourceSkeletonModifier() -> (SourceSkeletonModifier:Skeleton Modifier Object Reference)) -> (ReturnValue:Array of Names)
   n9 = String To Name(InString=Get BoneToTransfer() -> (BoneToTransfer:String))
   > n10 = Branch(Condition="integer >= integer"(A=Find Item!(TargetArray:Array of Names=@n8, ItemToFind:"Name (by ref)"=@n9) -> (ReturnValue)))
   > PrintString(InString=Append(A=Get BoneToTransfer(), B=" Found"))
   n11 = GetAllBoneNames(self=Get TargetSkeletonModifier())
   n12 = String To Name(InString=Get BoneToTransfer())
   > n13 = Branch(Condition="integer >= integer"(A=Find Item(TargetArray=@n11, ItemToFind=@n12)))
   n14 = String To Name(InString=Get BoneToTransfer())
   n15 = Get SourceSkeletonModifier()
   n16 = Get TargetSkeletonModifier()
   > Branch(Condition="Equal (Name)"(
     A=GetParentName!(self:Skeleton Modifier Object Reference=@n15, InBoneName:Name=@n14) -> (ReturnValue:Name=None)
     B=GetParentName(self=@n16, InBoneName=@n14) -> (ReturnValue=None)
    ))
   n17 = Get BoneToTransfer()
   > SetBoneTransform!(
    execute
    self:Skeleton Modifier Object Reference=Get TargetSkeletonModifier()
    InBoneName:Name=String To Name(InString=Get BoneToTransfer())
    InNewTransform:"Transform (by ref)"=GetBoneTransform!(self:Skeleton Modifier Object Reference=Get SourceSkeletonModifier(), InBoneName:Name=String To Name(InString=@n17), bGlobal:bool=false) -> (ReturnValue:Transform)
    bMoveChildren:bool=false
   ) -> (then, ReturnValue:bool)
   > PrintString(InString=Append!(A="Placed \"", B=Get BoneToTransfer(), C:String="\" in the Target Skeleton to match the Source Skeleton.") -> (ReturnValue))
   n18 = String To Name(InString=Get BoneToTransfer())
   n19 = Get SourceSkeletonModifier()
   AddBone!(
    execute=@n13.else
    self:Skeleton Modifier Object Reference=Get TargetSkeletonModifier()
    InBoneName:Name=@n18
    InParentName:Name=GetParentName(self=@n19, InBoneName=@n18) -> (ReturnValue=None)
    InTransform:"Transform (by ref)"=GetBoneTransform(self=@n19, InBoneName=@n18, bGlobal=false)
   ) -> (then, ReturnValue:bool)
   > PrintString(InString=Append(A="Added \"", B=Get BoneToTransfer(), C="\" to the Target Skeleton"))
   PrintString(execute=@n10.else, InString=Append(A="Could not find \"", B=Get BoneToTransfer(), C="\" in the Source Skeleton"))
   Event PreConstruct()
   > SetObject(self=Get DetailsView() -> (DetailsView:Details View Object Reference), NewObject=Self-Reference())
  ]}]
variables [
 BonesToTransfer {type="array<string>", default []}
 BoneToTransfer {type=string, default=""}
 SourceSkeletonModifier {type=SkeletonModifier, default=null}
 SourceSkelMesh {type=SkeletalMesh, default=null}
 TargetSkeletonModifier {type=SkeletonModifier, default=null}
 TargetSkelMesh {type=SkeletalMesh, default=null}
]
widgets {
 info {rootWidgetClass=&StackBox, widgetCount=7}
 widgets [
  StackBox_0 {widgetClassPath=&StackBox, widget=StackBox_0}
  VerticalBox_47 {
   widgetClassPath=&VerticalBox
   parent=StackBox_0
   widget=VerticalBox_47
   slot {class=StackBoxSlot, properties {size {sizeRule=Fill, value=1}}}
  }
  DetailsView {
   widgetClassPath=&DetailsView
   parent=VerticalBox_47
   widget=DetailsView
   properties {categoriesToShow [SkelMeshes]}
   slot {class=VerticalBoxSlot}
  }
  BoneList_TextBox {
   widgetClassPath=/Script/Blutility.EditorUtilityMultiLineEditableTextBox
   bIsVariable=true
   parent=VerticalBox_47
   widget=BoneList_TextBox
   properties {hintText="Copy and Paste your bone list from a selection of joints in your Source SkelMesh. \r\n\r\nIf the joint doesnt exist in the Source SkelMesh, it will error out. \r\n\r\nIf the joint is already in the Target SkelMesh, this tool will make sure that it's placement in space matches the Source SkelMesh."}
   slot {class=VerticalBoxSlot, properties {size {sizeRule=Fill, value=1}}}
  }
  HorizontalBox_76 {widgetClassPath=&HorizontalBox, parent=VerticalBox_47, widget=HorizontalBox_76, slot {class=VerticalBoxSlot}}
  TransferJointsButton {
   widgetClassPath=&EditorUtilityButton
   parent=HorizontalBox_76
   widget=TransferJointsButton
   slot {class=HorizontalBoxSlot, properties {padding=<ltrb 10 10 10 10>, size {sizeRule=Fill, value=1}}}
  }
  CommonTextBlock_53 {
   widgetClassPath=&CommonTextBlock
   parent=TransferJointsButton
   widget=CommonTextBlock_53
   properties {
    style=&TextStyle-Base_C
    text=Transfer Joints
    font {
     bForceMonospaced=false
     bMaterialIsStencil=false
     fontMaterial=null
     fontObject=&NotoSans
     letterSpacing=0
     monospacedWidth=1
     size=24
     skewAmount=0
     typefaceFontName=NotoSans-Regular
     outlineSettings {
      bApplyOutlineToDropShadows=false
      bMiteredCorners=false
      bSeparateFillAlpha=false
      outlineMaterial=null
      outlineSize=0
      outlineColor=<rgba 0 0 0 1>
     }
    }
    shadowOffset=<0 0>
   }
   slot {class=ButtonSlot}
  }
 ]
}
