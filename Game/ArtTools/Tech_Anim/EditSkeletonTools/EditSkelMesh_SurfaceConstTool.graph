EditorUtilityWidgetBlueprint /Game/ArtTools/Tech_Anim/EditSkeletonTools/EditSkelMesh_SurfaceConstTool
graphs [EventGraph {nodes [
   n1 = "On Selection Changed (ComboBox)"!() -> (OutputDelegate, then, SelectedItem:String, SelectionType:ESelectInfo Enum) {is=title K2Node_ComponentBoundEvent}
   > n2 = Set ConstraintTypeString(ConstraintTypeString:String=@n1.SelectedItem) -> (Output_Get:String)
   n3 = PrintString(InString=@n2)
   n4 = Branch(
    execute=[@n3.then, @n2.then]
    Condition="Equal, Case Insensitive (String)"(A:String=@n2, B:String=AnimBP Constraint) -> (ReturnValue:bool=~) {is=title K2Node_PromotableOperator}
   )
   > n5 = Set ConstraintTypeInt(ConstraintTypeInt:int=0) -> (Output_Get:int=0)
   n6 = Set ConstraintTypeInt(execute=@n4.else, ConstraintTypeInt=1) -> (Output_Get=0)
   n7 = Append!(
    A=Get generate_constraint_node() -> (generate_constraint_node:String)
    B=Get SkelMesh_ObjPath() -> (SkelMesh_ObjPath:String)
    C:String="', "
    D:String=Get BoneListAsArray() -> (BoneListAsArray:String)
    E:String=", "
    F:String="To String (Integer)"(InInt=Get ConstraintTypeInt() -> (ConstraintTypeInt:int=0))
    G:String=")"
   ) -> (ReturnValue)
   n8 = PrintString(InString=@n7)
   n9 = Set FinalPythonCommand(execute=@n12.then_3, FinalPythonCommand:String=@n7) -> (Output_Get:String)
   n10 = Execute Python Script!(execute=[@n8.then, @n9.then], PythonScript=@n9) -> (then, ReturnValue, returnstring:String) {outputs [returnstring]}
   Set Text(
    execute=[@n11.then, @n10.then]
    Text="To Text (String)"(InString=@n10.returnstring)
    self:"Text Box (Multi- Line) Object Reference"=Get Output_TextBox() -> (Output_TextBox:Editor Utility Multi Line Editable Text Box Object Reference)
   )
   n11 = PrintString(InString=@n10.returnstring)
   "On Clicked (CreateConstButton)"() {is=title K2Node_ComponentBoundEvent}
   > n12 = Sequence!(execute) -> (then_0, then_1, then_2:Exec, then_3:Exec)
   n13 = Get Text(self:"Text Box (Multi- Line) Object Reference"=Get BoneList_TextBox() -> (BoneList_TextBox:Editor Utility Multi Line Editable Text Box Object Reference))
   > Set BonesForConstraints(BonesForConstraints:Array of Strings=ParseIntoArray(SourceString="To String (Text)"(InText=@n13), Delimiter="\r\n")) -> (Output_Get:Array of Strings)
   n14 = Get Object Path String(Object=Get SkelMesh() -> (SkelMesh:Skeletal Mesh Object Reference))
   > Set SkelMesh_ObjPath(SkelMesh_ObjPath:String=@n14) -> (Output_Get:String)
   n15 = Get BonesForConstraints() -> (BonesForConstraints:Array of Strings)
   n16 = For Each Loop(Exec=@n12.then_1, Array=@n15)
   > n17 = Set BoneForConstraint(BoneForConstraint:String=@n16) -> (Output_Get:String)
   Clear(execute=@n12.then_2, TargetArray=Get TempBonesList() -> (TempBonesList:Array of Strings))
   > n18 = For Each Loop(Array=@n15)
   n19 = Get TempBonesList()
   n20 = Set BoneListAsArray(execute=@n18.Completed, BoneListAsArray:String=Append!(A="[", B=JoinStringArray(SourceArray=@n19, Separator=", "), C="]") -> (ReturnValue)) -> (Output_Get:String)
   Add(execute=@n18.LoopBody, TargetArray=Get TempBonesList(), NewItem=Append(A="\"", B=@n18, C="\"")) {is=title K2Node_CallArrayFunction}
   Event PreConstruct()
   > SetObject(self=Get DetailsView() -> (DetailsView:Details View Object Reference), NewObject=Self-Reference())
   PrintString(InString=@n17)
   PrintString(InString=@n14)
   PrintString(InString=@n20)
   PrintString(InString="To String (Integer)"(InInt=@n6))
   PrintString(InString="To String (Integer)"(InInt=@n5))
  ]}]
variables [
 BoneForConstraint {type=string, default=""}
 BoneListAsArray {type=string, default=""}
 BonesForConstraints {type="array<string>", default []}
 ConstraintTypeInt {type=integer, default=1}
 ConstraintTypeString {type=string, default=""}
 FinalPythonCommand {type=string, default=""}
 generate_constraint_node {
  type=string
  default="import importlib\r\nimport TechAnim.tools.surfaceConstGenerator.surfaceConstGenerator as surf\r\nimportlib.reload(surf)\r\nreturnstring = surf.generate_constraint_nodes('"
 }
 SkelMesh {type=SkeletalMesh, default=null}
 SkelMesh_ObjPath {type=string, default=""}
 SkinWeightModifier {type=SkinWeightModifier, default=null}
 TempBonesList {type="array<string>", default []}
]
widgets {
 info {rootWidgetClass=&StackBox, widgetCount=11}
 widgets [
  StackBox_0 {widgetClassPath=&StackBox, widget=StackBox_0}
  VerticalBox_47 {
   widgetClassPath=&VerticalBox
   parent=StackBox_0
   widget=VerticalBox_47
   slot {class=StackBoxSlot, properties {size {sizeRule=Fill, value=1}}}
  }
  DetailsView {
   widgetClassPath=&DetailsView
   parent=VerticalBox_47
   widget=DetailsView
   properties {categoriesToShow [SkelMeshes]}
   slot {class=VerticalBoxSlot}
  }
  ComboBox {
   widgetClassPath=/Script/Blutility.EditorUtilityComboBoxString
   bIsVariable=true
   parent=VerticalBox_47
   widget=ComboBox
   properties {
    selectedOption=ControlRig Constraint
    toolTipText="Pick the type of constraint you want to make.\r\n\r\nControlRig will create parent constraints that you can use in a ControlRig asset.\r\n\r\nAnimBP will create parent constraints that you can use in your Anim Blueprint Anim Graph."
    defaultOptions [ControlRig Constraint, AnimBP Constraint]
    font {
     bForceMonospaced=false
     bMaterialIsStencil=false
     fontMaterial=null
     fontObject=&Roboto
     letterSpacing=0
     monospacedWidth=1
     size=12
     skewAmount=0
     typefaceFontName=Bold
     outlineSettings {
      bApplyOutlineToDropShadows=false
      bMiteredCorners=false
      bSeparateFillAlpha=false
      outlineMaterial=null
      outlineSize=0
      outlineColor=<rgba 0 0 0 1>
     }
    }
    foregroundColor {colorUseRule=UseColor_Specified, specifiedColor=<rgba 0.5271 0.5271 0.5271 1>}
   }
   slot {class=VerticalBoxSlot}
  }
  VerticalBox_83 {widgetClassPath=&VerticalBox, parent=VerticalBox_47, widget=VerticalBox_83, slot {class=VerticalBoxSlot}}
  BoneList_TextBox {
   widgetClassPath=/Script/Blutility.EditorUtilityMultiLineEditableTextBox
   bIsVariable=true
   parent=VerticalBox_83
   widget=BoneList_TextBox
   properties {
    hintText="Copy and Paste your bone list from a selection of bones in your SkelMesh. \r\n\r\nThis tool will use that bone and try to find the closest vert on the surface of the Skel Mesh.  It will use that vert to create the code needed for a surface constraint(s).\r\n\r\nOne surface constraint will be made for each bone."
    toolTipText="Copy and Paste your bone list from a selection of bones in your SkelMesh. \r\n\r\nThis tool will use that bone and try to find the closest vert on the surface of the Skel Mesh.  It will use that vert to create the code needed for a surface constraint(s).\r\n\r\nOne surface constraint will be made for each bone."
   }
   slot {class=VerticalBoxSlot, properties {size {sizeRule=Fill, value=1}}}
  }
  Spacer_83 {widgetClassPath=&Spacer, parent=VerticalBox_47, widget=Spacer_83, properties {size=<1 5>}, slot {class=VerticalBoxSlot}}
  Output_TextBox {
   widgetClassPath=&EditorUtilityMultiLineEditableTextBox
   bIsVariable=true
   parent=VerticalBox_47
   widget=Output_TextBox
   properties {hintText=Your Code will be generated here..., toolTipText=Your Code will be generated here...}
   slot {class=VerticalBoxSlot, properties {size {sizeRule=Fill, value=1}}}
  }
  HorizontalBox_76 {widgetClassPath=&HorizontalBox, parent=VerticalBox_47, widget=HorizontalBox_76, slot {class=VerticalBoxSlot}}
  CreateConstButton {
   widgetClassPath=&EditorUtilityButton
   parent=HorizontalBox_76
   widget=CreateConstButton
   slot {class=HorizontalBoxSlot, properties {padding=<ltrb 10 10 10 10>, size {sizeRule=Fill, value=1}}}
  }
  CommonTextBlock_53 {
   widgetClassPath=&CommonTextBlock
   parent=CreateConstButton
   widget=CommonTextBlock_53
   properties {
    style=&TextStyle-Base_C
    text=Create Code
    font {
     bForceMonospaced=false
     bMaterialIsStencil=false
     fontMaterial=null
     fontObject=&NotoSans
     letterSpacing=0
     monospacedWidth=1
     size=24
     skewAmount=0
     typefaceFontName=NotoSans-Regular
     outlineSettings {
      bApplyOutlineToDropShadows=false
      bMiteredCorners=false
      bSeparateFillAlpha=false
      outlineMaterial=null
      outlineSize=0
      outlineColor=<rgba 0 0 0 1>
     }
    }
    shadowOffset=<0 0>
   }
   slot {class=ButtonSlot}
  }
 ]
}
